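# Physics asset for the F_MED_Flower_Skeleton backpack, defining collision bodies and joints for its left and right flower bones and stem bones.
# Each bone has a single sphere body, including the central pack bone.
# All eleven dyn_ flower and stem bodies constrain directly to pack in a star layout rather than parent-to-child chains.
# Every joint shares the same swing limit angles and a narrower twist limit angle.
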PhysicsAsset /Game/Accessories/FORT_Backpacks/Backpack_F_MED_Flower_Skeleton/Meshes/F_MED_Flower_Skeleton_Physics
bodies {
 dyn_flower_l_a [{shapeType=Sphere, radius=2, center=<2 0 0>}]
 dyn_flower_l_b [{shapeType=Sphere, radius=2, center=<2 0 0>}]
 dyn_flower_l_c [{shapeType=Sphere, radius=2, center=<2 0 0>}]
 dyn_flower_l_d [{shapeType=Sphere, radius=2.8021, center=<0.5977 -3 0>}]
 dyn_flower_r_a [{shapeType=Sphere, radius=2, center=<-2 0 0>}]
 dyn_flower_r_b [{shapeType=Sphere, radius=2, center=<0 2 0>}]
 dyn_flower_r_c [{shapeType=Sphere, radius=2, center=<0 2 0>}]
 dyn_flower_r_d [{shapeType=Sphere, radius=2, center=<0 2 0>}]
 dyn_stem_a [{shapeType=Sphere, radius=2, center=<0 2 0>}]
 dyn_stem_b [{shapeType=Sphere, radius=2, center=<0 2 0>}]
 dyn_stem_c [{shapeType=Sphere, radius=2, center=<0 2 0>}]
 pack [{shapeType=Sphere, radius=3.2149, center=<0 -0.3171 0>}]
}
constraints [
 {bone1Name=dyn_flower_l_a, bone2Name=pack, swing1LimitDegrees=25, swing2LimitDegrees=25, twistLimitDegrees=10}
 {bone1Name=dyn_flower_l_b, bone2Name=pack, swing1LimitDegrees=25, swing2LimitDegrees=25, twistLimitDegrees=10}
 {bone1Name=dyn_flower_l_c, bone2Name=pack, swing1LimitDegrees=25, swing2LimitDegrees=25, twistLimitDegrees=10}
 {bone1Name=dyn_flower_l_d, bone2Name=pack, swing1LimitDegrees=25, swing2LimitDegrees=25, twistLimitDegrees=10}
 {bone1Name=dyn_flower_r_a, bone2Name=pack, swing1LimitDegrees=25, swing2LimitDegrees=25, twistLimitDegrees=10}
 {bone1Name=dyn_flower_r_d, bone2Name=pack, swing1LimitDegrees=25, swing2LimitDegrees=25, twistLimitDegrees=10}
 {bone1Name=dyn_flower_r_c, bone2Name=pack, swing1LimitDegrees=25, swing2LimitDegrees=25, twistLimitDegrees=10}
 {bone1Name=dyn_flower_r_b, bone2Name=pack, swing1LimitDegrees=25, swing2LimitDegrees=25, twistLimitDegrees=10}
 {bone1Name=dyn_stem_a, bone2Name=pack, swing1LimitDegrees=25, swing2LimitDegrees=25, twistLimitDegrees=10}
 {bone1Name=dyn_stem_b, bone2Name=pack, swing1LimitDegrees=25, swing2LimitDegrees=25, twistLimitDegrees=10}
 {bone1Name=dyn_stem_c, bone2Name=pack, swing1LimitDegrees=25, swing2LimitDegrees=25, twistLimitDegrees=10}
]
